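# Sound cue for music-party music at the Helios SteppedTown point of interest, playing a temporary acoustic guitar loop borrowed from MountainVillage.
# A looping wave player feeds an Enveloper whose volume curve fades the sound in over the first second.
# The base submix is disabled and the cue sends at full level to the MusicParty SteppedTown audio bus, so it serves as a bus feed.
# A concurrency override caps simultaneous instances.
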SoundCue /Game/Sounds/Helios/POI/SteppedTown/Cue_MusicParty_MS_Helios_SteppedTown
root=Enveloper_0
nodes [
 {
  id=Enveloper_0
  children [WavePlayer_0]
  properties {pitchCurve {editorCurveData {keys [{time=0}]}}, volumeCurve {editorCurveData {keys [<at 0 0>,<at 1 1>]}}}
 }
 {id=WavePlayer_0, properties {bLooping=true, soundWaveAssetPtr=/Game/Sounds/Helios/POI/MountainVillage/Temp_Music_Acoustic_Guitar_Futz_Loop}}
]
properties {
 bEnableBaseSubmix=false
 bOverrideConcurrency=true
 maxDistance=2097152
 volumeMultiplier=1
 busSends [{audioBus=&/MusicParty_MS_Helios_SteppedTown, minSendLevel=0, sendLevel=1, soundSourceBus=null}]
 concurrencyOverrides {maxCount=4}
}
